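; Weapon Blueprint for the Thorn Ripper, a chain whip with a melee attack and a hookshot.
; It attaches the mesh and sets HUD actions on equip, removes the local camera modifier on unequip or EndPlay, and stores an authority-only hookshot projectile reference.
; An async targeting trace feeds hit distance to the AnimBP to scale chain length.
; Hookshot targeting swaps sounds and HUD one tick late, simulating a melee press on exit if primary is held outside vehicles.
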(event FortWeapon|EventOnEquippedAndAttached
  ;[0.02491 0.1042 0.02354] On Equip: Set up HUD key actions (shown on the left hand side of the screen) and attach the weapon mesh to the desired socket on the hand of the pawn.
  (|CallParentFunction)
  (CallFunction|UpdateHudActions false)
  (bind _asfort_pawn (Utilities|Casting|CastToFortPawn (Game|GetInstigator)))
  ;;[0.02934 0.1667 0.16] Attach the weapon mesh to the appropriate player mesh socket.
  (bind _mesh (Variables|/Script/Engine.Character|GetMesh _asfort_pawn)
    (:then
      (Transformation|AttachComponentToComponent (Variables|Visual|GetWeaponMesh) _mesh (Variables|WeaponPositionInfo|GetMainhandSocketName) "SnapToTarget" "SnapToTarget" "KeepWorld" true))
    (:else)))

(event FortWeapon|EventOnUnEquippedAndDetached
  ;[0.0243 0.165 0.1667] OnUnequippedAndDetached: Fade out the looping idle sound and disable/remove the camera modifier.
  (|CallParentFunction)
  (Utilities|FlowControl|Sequence
    (:then_0
      (Audio|SpawnSoundAttached "/SunRoseChainWhipGameplay/Items/ChainWhip/Sound/ChainWhip_UnEquip_Cue.ChainWhip_UnEquip_Cue" (Variables|/Script/Engine.Actor|GetRootComponent (Game|GetInstigator)) "hand_r"))
    (:then_1
      (Networking|CanExecuteCosmeticEvents
        (:True
          (bind _cameramodifier (Variables|Camera|GetCameraModifier)
            (:then
              (Class|PlayerCameraModifierChainWhip|DisableandRemove _cameramodifier)
              (Variables|Camera|SetCameraModifier 0))
            (:else)) ; Only exists on local player
        )))))

(event Custom|UpdateAttackRangeAnimBP
  (bind _returnvalue (GameInstanceSubsystems|GetTargetingSubsystem))
  ;[0.09147 0.03613 0.1094] UpdateAttackRangeAnimBP: Perform a targeting trace to figure out the distance of the trace in the aiming direction.
  ; This value is used in the animation blueprint to scale the length of the chain on the attack.
  ; Note: This logic needs to run on simulated and autonomous proxies.
  (Networking|CanExecuteCosmeticEvents
    (:True
      (bind _b_chainwhiptargetingcomponent (Variables|Default|GetB_ChainWhipTargetingComponent)
        (:then
          (Utilities|IsValid (General|GetPlayerPawn _b_chainwhiptargetingcomponent)
            (:Is%20Valid
              (bind (_asynctaskproxy _targetinghandle) (Class|AsyncActionPerformTargeting|PerformTargetingAsyncAction (General|GetPlayerPawn _b_chainwhiptargetingcomponent) "/SunRoseChainWhipGameplay/Items/ChainWhip/Abilities/Melee/Targeting/TP_ChainWhip_MeleeUpdateRange.TP_ChainWhip_MeleeUpdateRange")
                (:then
                  (Variables|Transient|SetAsyncTask_AttackRangeTrace _asynctaskproxy))
                (:Targeted
                  (Utilities|IsValid _returnvalue
                    (:Is%20Valid
                      (if (Utilities|Array|IsValidIndex (TargetingSystem|TargetingResults|GetTargetingResults _returnvalue _targetinghandle) (Utilities|Array|LastIndex (TargetingSystem|TargetingResults|GetTargetingResults _returnvalue _targetinghandle)))
                        (if (Collision|BreakHitResult (Utilities|Array|Get(aref) (TargetingSystem|TargetingResults|GetTargetingResults _returnvalue _targetinghandle) (Utilities|Array|LastIndex (TargetingSystem|TargetingResults|GetTargetingResults _returnvalue _targetinghandle))))
                          ;;[0.03005 0.151 0.1253] Update the attack range based on the distance result of the trace.
                          (bind _animbp (Utilities|GetWeaponAnimBP)
                            (:Valid
                              (Class|ChainsSunRoseSharpWepAnimBP|UpdateAttackRange _animbp (.Distance (Collision|BreakHitResult (Utilities|Array|Get(aref) (TargetingSystem|TargetingResults|GetTargetingResults _returnvalue _targetinghandle) (Utilities|Array|LastIndex (TargetingSystem|TargetingResults|GetTargetingResults _returnvalue _targetinghandle))))))))
                          (else
                            _))
                        (else
                          _))
                      ;;[0.06385 0.194 0.1979] Any failure to perform the trace should update the attack range to max distance.
                      (bind _animbp_2 (Utilities|GetWeaponAnimBP)
                        (:Valid
                          (Class|ChainsSunRoseSharpWepAnimBP|UpdateAttackRange _animbp_2 (ScalableFloat|GetValueatLevel (Variables|Hotfix|GetHotfix_MeleeAttackRange))))))
                    (:Is%20Not%20Valid
                      _)))))
            (:Is%20Not%20Valid
              _)))
        (:else
          _)))))

(event OnTargetingHookshot(B_ChainWhipTargetingComponent)
  (params (bTargetingHookshot :type Boolean))
  (bind _returnvalue (Game|GetInstigator))
  (bind _weaponequipsoundcomponent (Variables|Default|GetWeaponEquipSoundComponent))
  (bind _soundlibrarycomponent (Variables|SoundLibrary|GetSoundLibraryComponent))
  ;[0.2917 0.161 0.111] OnTargetingHookshot: Play the appropriate sound when the weapon is entering and exiting the hookshot targeting state.
  (Networking|CanExecuteCosmeticEvents
    (:True
      (if bTargetingHookshot
        ;;[0.1976 1 0.9801] This event is expensive (~300 microseconds) and will occur during frames where the chainwhip is already expensive, so we delay 1 frame to avoid a spike in cost.
        (Utilities|FlowControl|DelayUntilNextTick)
        (SoundLibrary|PlayEvent _soundlibrarycomponent (Variables|Sounds|GetTargetingSoundLibraryTag))
        (Class|WeaponEquipSoundComponent|StopAudio _weaponequipsoundcomponent)
        (CallFunction|UpdateHudActions true)
        (else
          (Utilities|FlowControl|DelayUntilNextTick)
          (SoundLibrary|PlayEvent _soundlibrarycomponent (Variables|Sounds|GetTargetingExitSoundLibraryTag))
          (Class|WeaponEquipSoundComponent|PlayAudio _weaponequipsoundcomponent)
          (CallFunction|UpdateHudActions false)
          ;;[0.1531 0.1804 1] If the player is holding down the primary input when targeting ends (and NOT in a vehicle), perform a melee attack.
          (bind _asfort_player_pawn (Utilities|Casting|CastToFortPlayerPawn _returnvalue))
          (if (Math|Boolean|ANDBoolean (Variables|/Script/FortniteGame.FortPawn|GetbPrimaryInputHeld _asfort_player_pawn) (Pawn|IsLocallyControlled _asfort_player_pawn) (not (Vehicle|IsinVehicle _asfort_player_pawn)))
            (Mantis|SimulateInputEvent (Actor|GetComponentbyClass _returnvalue "/Script/MantisRuntime.FortMantisPawnComponent") "Primary" "Pressed" true false)))))))

(event EventEndPlay
  (params (EndPlayReason :type EEndPlayReason%20Enum :default "Destroyed"))
  ;[0.1198 0.06687 0.02099] EndPlay: Clean up the added camera modifier.
  (Utilities|FlowControl|Sequence
    (:then_0
      (Networking|CanExecuteCosmeticEvents
        (:True
          (bind _cameramodifier (Variables|Camera|GetCameraModifier)
            (:then
              (Class|PlayerCameraModifierChainWhip|DisableandRemove _cameramodifier)
              (Variables|Camera|SetCameraModifier 0))
            (:else)) ; Only exists on local player
        )))
    (:then_1
      (|CallParentFunction EndPlayReason))))

(fn GetCameraModifier
  (return (Variables|Camera|GetCameraModifier)))

(fn GetHookshotProjectile
  ;[0.04513 0.2031 0.1829] Return the reference to the latest hookshot projectile. Can be null. Client always returns null.
  (return (Variables|Hookshot|GetHookshotProjectile)))

(fn SetHiddenMeleeTrail
  (params (NewHidden :type Boolean))
  ;[0.05441 0.126 0.1354] SetHiddenMeleeTrail: Hide/Unhide the melee trail component when desired.
  (Development|SetHiddeninGame (Variables|Default|GetNS_ChainWhip_Melee_Trail) NewHidden false))

(fn SetHookshotProjectile
  (params (HookshotProjectile :type B%20Prj%20Thorn%20Ripper%20Hookshot%20V%202%20Object%20Reference))
  ;[0.03895 0.1819 0.1875] SetHookshotProjectile: Called to store the ref to the projectile so that it can be accessed from other locations.
  ; Called from GA_ChainWhip_Secondary_Hookshot.
  (Networking|SwitchHasAuthority
    (:Authority
      (Variables|Hookshot|SetHookshotProjectile HookshotProjectile))))

(fn UpdateHudActions
  (params (InTargetingMode :type Boolean))
  (WeaponHUDData|SetHudKeyActionsVisibilitybyID (Utilities|Array|MakeArray (Utilities|Struct|MakeWeaponHudKeyActionIdandVisibility "ChainWhip_Hookshot_Fire" InTargetingMode) (Utilities|Struct|MakeWeaponHudKeyActionIdandVisibility "ChainWhip_Attack" (not InTargetingMode)) (Utilities|Struct|MakeWeaponHudKeyActionIdandVisibility "ChainWhip_Hookshot_Target" (not InTargetingMode)))))

(fn ConstructionScript
  (|CallParentFunction))
